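; Gameplay ability that gives a small winged creature AI pawn a chance-based airborne jump.
; It can activate only when the pawn has a goal actor and the blackboard goal lies within a scalable distance range.
; On a successful roll it zeroes gravity, launches the pawn upward, then along its forward vector, and ends after a random duration.
; Ending the ability restores normal gravity on the stored pawn.
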(event Ability|EventActivateAbility
  (bind _aipawn (Variables|Default|GetAIPawn))
  (Ability|CommitAbility)
  (if (<= (Math|Random|RandomFloat) (ScalableFloat|GetValueatLevel (Variables|Default|GetChanceToJump)))
    (Utilities|FlowControl|Delay 0.2)
    (bind _asfort_pawn (Utilities|Casting|CastToFortPawn (Ability|GetAvatarActorfromActorInfo)))
    (bind _output_get (Variables|Default|SetAIPawn _asfort_pawn))
    (FortPawn|SetGravityMultiplier _output_get 0.0)
    (Character|LaunchCharacter _aipawn (Math|Vector|MakeVector 0.0 0.0 (ScalableFloat|GetValueatLevel (Variables|Default|GetJumpUpwardVelocity))) true true)
    (Utilities|FlowControl|Delay 0.5)
    (Character|LaunchCharacter _aipawn (* (Transformation|GetActorForwardVector _aipawn) (ScalableFloat|GetValueatLevel (Variables|Default|GetJumpForwardVelocity))) true true)
    (Utilities|FlowControl|Delay (Math|Random|RandomFloatinRange (ScalableFloat|GetValueatLevel (Variables|Default|GetJumpDurationMin)) (ScalableFloat|GetValueatLevel (Variables|Default|GetJumpDurationMax))))
    (Ability|EndAbility)
    (else
      (Ability|EndAbility))))

(event Ability|EventOnEndAbility
  (params (bWasCancelled :type Boolean))
  (bind _aipawn (Variables|Default|GetAIPawn))
  (Utilities|IsValid _aipawn
    (:Is%20Valid
      (FortPawn|SetGravityMultiplier _aipawn 1.0))))

(fn CanActivateAbility
  (params (ActorInfo :type Gameplay%20Ability%20Actor%20Info%20Structure) (Handle :type Gameplay%20Ability%20Spec%20Handle%20Structure))
  (bind _asfort_aipawn (Utilities|Casting|CastToFortAIPawn (Ability|GetAvatarActorfromActorInfo)))
  (if (Variables|/Script/FortniteGame.FortAIPawn|GetbHasGoalActor _asfort_aipawn)
    (bind _asactor (Utilities|Casting|CastToActor (AI|Components|Blackboard|GetValueasObject (AI|GetBlackboard (Pawn|GetController _asfort_aipawn)) (LocalVariables|GetGoalActorBBKeyName))))
    (return 0 (and (>= (Transformation|GetDistanceTo (Ability|GetAvatarActorfromActorInfo) _asactor) (ScalableFloat|GetValueatLevel (Variables|Default|GetMinimumJumpDistance))) (<= (Transformation|GetDistanceTo (Ability|GetAvatarActorfromActorInfo) _asactor) (ScalableFloat|GetValueatLevel (Variables|Default|GetMaximumJumpDistance)))))
    (else
      (return 0 false))))
